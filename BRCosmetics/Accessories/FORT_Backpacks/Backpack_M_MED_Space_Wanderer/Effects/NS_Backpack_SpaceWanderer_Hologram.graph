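# Niagara system for the Space Wanderer backpack hologram: a persistent, slowly rotating holographic effect with a cyan glow.
# LightGlow drives sprite, light and mesh renderers, while Sphere renders randomly oriented meshes at one of two small scales.
# Each emitter bursts two particles on its first loop only, which live until the emitter deactivates.
# Random rotation forces, X-only for LightGlow and all axes for Sphere, feed a clamped rotational velocity, producing bounded, wandering spin.
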NiagaraSystem /BRCosmetics/Accessories/FORT_Backpacks/Backpack_M_MED_Space_Wanderer/Effects/NS_Backpack_SpaceWanderer_Hologram
dependencies {usedDataInterfaces [], usedDynamicInputs [], usedModules [&SystemState]}
emitters {
 LightGlow {
  emitterUpdateScript [
   &EmitterState()
   &SpawnBurst_Instantaneous(
    Spawn Count=&MakeIntFromFloat(Float=2)
    Spawn Probability=HlslExpression() {hlslExpression="Emitter.LoopCount == 0"}
    Use Spawn Probability=true
   )
  ]
  particleSpawnScript [
   /Niagara/Modules/Spawn/Initialization/InitializeParticle(
    Color=<rgba 0 0.9 1 0.35>
    Mesh Scale=<21 21 21>
    Module.Write Color=true
    Module.Write Lifetime=true
    Module.Write Mass=true
    Module.Write Material Random=true
    Module.Write Position=true
    Module.Write Scale=true
    Module.Write SpriteRotation=false
    Module.Write SpriteSize=true
    Position=&SimulationPosition()
    Sprite Size=&Vector2DFromFloat(Value=20)
   )
   /SetVariables(Particles.MeshOrientation=&OrientMeshToVector(Look At Direction=<0 0 -1>, Reference Axis Vector=<1 0 0>)) {isSetParameters=true}
   /SetVariables(Particles.RotationalVelocity:Vector3f=&UniformRangedVector(Minimum=<-0.4 0 0>, Maximum=<0.4 0 0>)) {isSetParameters=true}
   /SetVariables(Particles.LightColor=<rgba 0 4 3.8 1>) {isSetParameters=true}
   /SetVariables(Particles.LightExponent=2) {isSetParameters=true}
  ]
  particleUpdateScript [
   &ParticleState(Kill Particles When Lifetime Has Elapsed=false, Let Infinitely Lived Particles Die When Emitter Deactivates=true)
   &MeshRotationForce(Rotation=&UniformRangedVector(Minimum=<-1 0 0>, Maximum=<1 0 0>))
   &SolveForcesAndVelocity(Clamp Velocity=false, Limit Acceleration=false)
   &SolveRotationalForcesAndVelocity()
   /SetVariables(Particles.RotationLimited:Vector3f=&self:LimitVelocity(
     Max:Vector3f=&VectorFromFloat(Value=0.4)
     Min:Vector3f=&VectorFromFloat(Value=-0.4)
    )) {isSetParameters=true}
   /SetVariables(Particles.RotationalVelocity:Vector3f=$Particles.RotationLimited) {isSetParameters=true}
  ]
  rendererClasses [
   &NiagaraSpriteRendererProperties
   &NiagaraLightRendererProperties
   &NiagaraMeshRendererProperties
  ]
  renderers [
   {rendererClass=&NiagaraSpriteRendererProperties}
   {rendererClass=&NiagaraLightRendererProperties, rendererIndex=1}
   {rendererClass=&NiagaraMeshRendererProperties, rendererIndex=2}
  ]
 }
 Sphere {
  emitterSpawnScript [&FN_PreserveSkydiveUserVariables() {module=PreserveUserVariables}]
  emitterUpdateScript [
   &EmitterState()
   &SpawnBurst_Instantaneous(
    Spawn Count=&MakeIntFromFloat(Float=2)
    Spawn Probability=HlslExpression() {hlslExpression="Emitter.LoopCount == 0"}
    Use Spawn Probability=true
   )
  ]
  particleSpawnScript [
   &InitializeParticle(
    Mesh Scale=&VectorFromFloat(Value=&UniformAOrBFloat(A=0.14, B=0.155))
    Module.Write Color=true
    Module.Write Lifetime=true
    Module.Write Mass=false
    Module.Write Material Random=true
    Module.Write Position=false
    Module.Write Scale=true
    Module.Write SpriteRotation=false
    Module.Write SpriteSize=true
    Sprite Size=<19 19>
   )
   /SetVariables(Particles.MeshOrientation=&OrientMeshToVector(
     Look At Direction=&UniformRangedVector(Minimum=<-1 -1 -1>, Maximum=<1 1 1>)
     Reference Axis Vector=<1 0 0>
    )) {isSetParameters=true}
   /SetVariables(Particles.RotationalVelocity:Vector3f=&UniformRangedVector(
     Minimum=&VectorFromFloat(Value=-0.4)
     Maximum=&VectorFromFloat(Value=0.4)
    )) {isSetParameters=true}
   /SetVariables(Particles.SpritePosition:Vector3f=&Add_Vector(A=Linked() {linkedVariable {name=Particles.Position, type=NiagaraPosition}}, B=<0 0.5 -0.5>)) {isSetParameters=true}
  ]
  particleUpdateScript [
   &ParticleState(Kill Particles When Lifetime Has Elapsed=false, Let Infinitely Lived Particles Die When Emitter Deactivates=true)
   &MeshRotationForce(Rotation=&UniformRangedVector(Minimum=<-3 -3 -3>, Maximum=<3 3 3>))
   &SolveForcesAndVelocity(Clamp Velocity=false, Limit Acceleration=false)
   &SolveRotationalForcesAndVelocity()
   /SetVariables(Particles.RotationLimited:Vector3f=&self:LimitVelocity(
     Max:Vector3f=&VectorFromFloat(Value=0.4)
     Min:Vector3f=&VectorFromFloat(Value=-0.4)
    )) {isSetParameters=true}
   /SetVariables(Particles.RotationalVelocity:Vector3f=$Particles.RotationLimited) {isSetParameters=true}
  ]
  rendererClasses [&NiagaraMeshRendererProperties]
  renderers [{rendererClass=&NiagaraMeshRendererProperties, rendererIndex=0}]
 }
}
summary {
 systemName=NS_Backpack_SpaceWanderer_Hologram
 emitters [
  {
   emitterName=LightGlow
   rendererClasses [
    &NiagaraSpriteRendererProperties
    &NiagaraLightRendererProperties
    &NiagaraMeshRendererProperties
   ]
  }
  {emitterName=Sphere, rendererClasses [&NiagaraMeshRendererProperties]}
 ]
 userVariables [
  User.bIsFrontEndPreview {type=NiagaraBool}
  User.FEVec {type=Vector3f, description="", defaultValue {struct=Vector3f, value=<0 0 1000>}}
  User.LightVariant {type=LinearColor, description="", defaultValue {struct=LinearColor, value=<rgba 1 1 1 1>}}
  User.PlayerSpeed {type=NiagaraFloat, defaultValue=0, description=""}
  User.RibbonIDOffset {type=NiagaraInt32}
  User.SkeletalMesh {
   type=NiagaraDataInterfaceSkeletalMesh
   defaultValue {
    struct=NiagaraExt_VariableValue_DataInterface
    value {dataInterface=&self:NiagaraDataInterfaceSkeletalMesh_0}
   }
  }
  User.SpawnRate {type=NiagaraFloat, defaultValue=360, description=""}
 ]
}
