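# Physics asset for the Space Wanderer backpack cosmetic, giving its skeletal mesh simple collision bodies.
# A large sphere covers the dyn_glass bone, and a small zero-length capsule sits on the pack bone.
# A single constraint joins the two bones with both swings and twist locked, so the glass cannot rotate relative to the pack.
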PhysicsAsset /BRCosmetics/Accessories/FORT_Backpacks/Backpack_M_MED_Space_Wanderer/Meshes/M_MED_Space_Wanderer_Pack_Physics
bodies {
 dyn_glass [{shapeType=Sphere, radius=10.1936}]
 pack [{shapeType=Capsule, length=0, radius=1.6245, center=<-0.0692 -0.6791 -0.0167>, rotation=<rot 0 0 -90>}]
}
constraints [{
  bone1Name=dyn_glass
  bone2Name=pack
  swing1LimitDegrees=0
  swing1Motion=Locked
  swing2LimitDegrees=0
  swing2Motion=Locked
  twistLimitDegrees=0
  twistMotion=Locked
 }]
