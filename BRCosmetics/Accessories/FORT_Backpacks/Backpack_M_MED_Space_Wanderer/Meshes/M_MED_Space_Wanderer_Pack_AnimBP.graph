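# Animation Blueprint for the Space Wanderer backpack cosmetic, animating a spinning moon bone and gently floating top and bottom parts.
# Each update derives moon spin from real time, plus bobbing offsets and a shared wobble rotation from game-time sine and cosine waves.
# A chain of Modify Bone nodes applies these to the ref pose, then a RigidBody node simulates physics in root-bone space.
# Jump_Alpha is set from a jumping flag but is unused in the AnimGraph.
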AnimBlueprint /BRCosmetics/Accessories/FORT_Backpacks/Backpack_M_MED_Space_Wanderer/Meshes/M_MED_Space_Wanderer_Pack_AnimBP
target_skeleton=&/M_MED_Space_Wanderer_Pack_Skeleton
graphs [
 AnimGraph {nodes [
   n1 = "Transform (Modify) Bone - Bone: dyn_moon"(
    ComponentPose=Local To Component(LocalPose=Local Space Ref Pose())
    Alpha=1
    Translation="0,0,0"
    Rotation=Get SpinnerRot() -> (SpinnerRot:Rotator)
    Scale="1,1,1"
   ) {rotationMode=BMM_Replace, rotationSpace=BCS_BoneSpace, alphaBoolBlend {blendOption=Linear}, boneToModify {boneName=dyn_moon}}
   n2 = "Transform (Modify) Bone - Bone: dyn_float_top_02"(ComponentPose=@n1, Alpha=1, Translation=Get PackTrans() -> (PackTrans:Vector), Rotation="0,0,0", Scale="1,1,1") {translationMode=BMM_Replace, translationSpace=BCS_BoneSpace, alphaBoolBlend {blendOption=Linear}, boneToModify {boneName=dyn_float_top_02}}
   n3 = "Transform (Modify) Bone - Bone: dyn_float_bot_02"(ComponentPose=@n2, Alpha=1, Translation=Get PackTrans2() -> (PackTrans2:Vector), Rotation="0,0,0", Scale="1,1,1") {translationMode=BMM_Replace, translationSpace=BCS_BoneSpace, alphaBoolBlend {blendOption=Linear}, boneToModify {boneName=dyn_float_bot_02}}
   n4 = Get packRot() -> (packRot:Rotator)
   n5 = RigidBody(ComponentPose="Transform (Modify) Bone - Bone: dyn_float_bot_01"(
     ComponentPose="Transform (Modify) Bone - Bone: dyn_float_top_01"(ComponentPose=@n3, Alpha=1, Translation="0,0,0", Rotation=@n4, Scale="1,1,1") {rotationMode=BMM_Additive, alphaBoolBlend {blendOption=Linear}, boneToModify {boneName=dyn_float_top_01}}
     Alpha=1
     Translation="0,0,0"
     Rotation=@n4
     Scale="1,1,1"
    ) {rotationMode=BMM_Additive, alphaBoolBlend {blendOption=Linear}, boneToModify {boneName=dyn_float_bot_01}}) {
    bClampLinearTranslationLimitToRefPose=true
    bOverrideWorldGravity=true
    overridePhysicsAsset=&/M_MED_Space_Wanderer_Pack_Physics
    simulationSpace=BaseBoneSpace
    alphaBoolBlend {blendOption=Linear}
    baseBoneRef {boneName=root}
    componentLinearAccScale=<0.7 0.7 0.7>
    componentLinearVelScale=<0.7 0.7 0.7>
    overrideWorldGravity {z=-980}
   }
   Output Pose(Result=Component To Local(ComponentPose=@n5)) {layerGroup=None}
  ]}
 EventGraph {nodes [
   Event BlueprintUpdateAnimation()
   > n1 = Sequence!(execute) -> (then_0, then_1, then_2:Exec, then_3:Exec, then_4:Exec)
   n2 = float * float(A=float / float(A=GetRealTimeSeconds(), B=-720) {is=title K2Node_CallFunction}, B=40000)
   > Set SpinnerRot(SpinnerRot:Rotator=MakeRotator(Roll=@n2, Yaw=@n2)) -> (Output_Get:Rotator)
   n3 = GetGameTimeInSeconds()
   n4 = MapRangeClamped(Value="Cos (Degrees)"!(A:double=float * float(A=@n3, B=30)) -> (ReturnValue:double), InRangeA=-2, InRangeB=2, OutRangeA=-20, OutRangeB=20)
   Set packRot(execute=@n1.then_1, packRot:Rotator=MakeRotator(Roll=@n4, Pitch=@n4, Yaw=@n4)) -> (Output_Get:Rotator)
   n5 = "Sin (Degrees)"!(A:double=float * float(A=GetGameTimeInSeconds(), B=120)) -> (ReturnValue:double)
   Set PackTrans(execute=@n1.then_2, PackTrans:Vector=MakeVector(X=MapRangeClamped(Value=@n5, InRangeA=-1, InRangeB=1, OutRangeA=-0.1, OutRangeB=0.7), Y=0)) -> (Output_Get:Vector)
   n6 = "Sin (Degrees)"(A=float * float(A=GetGameTimeInSeconds(), B=90))
   Set PackTrans2(execute=@n1.then_3, PackTrans2:Vector=MakeVector(X=MapRangeClamped(Value=@n6, InRangeA=-1, InRangeB=1, OutRangeA=-0.1, OutRangeB=0.7), Y=0)) -> (Output_Get:Vector)
   Set Jump_Alpha(execute=@n1.then_4, Jump_Alpha:double=SelectFloat(A=0.75, B=0, bPickA=Get bIsJumping())) -> (Output_Get:double)
   TryGetPawnOwner()
   TryGetPawnOwner()
  ]}
]
variables [
 Jump_Alpha {type=number}
 packRot {type=Rotator}
 PackTrans {type=Vector}
 PackTrans2 {type=Vector}
 SpinnerRot {type=Rotator}
]
